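# Unlit UI material for the InstantGravel item shop tile, compositing a tile background with the StallionSmoke featured outfit image.
# Named reroutes define fill-scaled UVs from MF_UI_UVScaling, plus a dark blue exponential radial gradient and its mask.
# A first lerp combines the tile texture with rerouted inputs.
# A second lerp layers the outfit texture on top, using that texture's own RGB as the blend alpha.
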Material /ItemShop/UI/Art/InstantGravel/M_UI_InstantGravel_ShopTile
graph {
 properties {
  bCastRayTracedShadows=false
  bUsedWithNanite=true
  bUsedWithStaticMesh=false
  materialDomain=MD_UI
  pixelDepthOffsetMode=PDOM_Legacy
  shadingModel=MSM_Unlit
  usedShadingModels=Unlit
 }
 nodes [
  n1 = NamedRerouteUsage()
  n2 = TextureSample(UVs=@n1, Tex=TextureObject() {texture=/OfferCatalog/Textures/BattleRoyale/FeaturedItems/Outfit/S22/T-AthenaSoldiers-Character-StallionSmoke})
  LinearInterpolate(
   A=LinearInterpolate(
    A=TextureSample(UVs=@n1) {texture=/OfferCatalog/Art/InstantGravel/T-UI-InstantGravel-Tile}.RGB
    B=NamedRerouteUsage()
    Alpha=NamedRerouteUsage()
   )
   B=@n2.RGB
   Alpha=@n2.RGB
  ) {constAlpha=1}
  NamedRerouteDeclaration(None=Multiply(A=MaterialFunctionCall() {materialFunction=&MF_UI_UVScaling}.Scaled To Fill)) {name=UVFill, nodeColor=<rgba 1 0 0.7266 1>}
  n3 = MaterialFunctionCall(UVs=NamedRerouteUsage(), Radius=Constant() {r=0.8}, Density=Constant() {r=1.4}) {materialFunction=&RadialGradientExponential}
  NamedRerouteDeclaration(None=Multiply(A=Constant3Vector() {constant=<rgba 0.011 0.0115 0.0208 1>}, B=@n3.RadialGradientExponential)) {name=Gradient, nodeColor=<rgba 0.3828 0 1 1>}
  NamedRerouteDeclaration(None=@n3.RadialGradientExponential) {name=Gradient Alpha, nodeColor=<rgba 1 0 0.0937 1>}
 ]
}
